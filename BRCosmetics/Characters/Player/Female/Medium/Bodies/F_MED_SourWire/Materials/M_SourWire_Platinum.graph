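# Platinum body material for the SourWire female character cosmetic, layering an audio-reactive sequin and stripe effect over the shared CharacterParent 2023 shader.
# View-angle lookups into thin-film and platinum gradient textures supply iridescence, while a pre-skinned height mask separates the boots.
# Sequin textures, panning noise and sine-animated stripes shape surface detail, with emissive scaled by audio amplitude and PlatinumReactivity.
# The tech art mask blends these attributes between PreFX and PostFX functions.
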Material /BRCosmetics/Characters/Player/Female/Medium/Bodies/F_MED_SourWire/Materials/M_SourWire_Platinum
graph {
 properties {
  bHasPixelAnimation=true
  bUseMaterialAttributes=true
  bUsedWithClothing=true
  bUsedWithMeshDeformer=true
  bUsedWithMorphTargets=true
  bUsedWithNanite=true
  bUsedWithNiagaraMeshParticles=true
  bUsedWithSkeletalMesh=true
  pixelDepthOffsetMode=PDOM_Legacy
  shadingModel=MSM_FromMaterialExpression
  usedShadingModels=Default Lit | Subsurface | Clear Coat | Subsurface Profile | Cloth
 }
 nodes [
  n1 = MaterialFunctionCall() {materialFunction=&MF_CharacterParent_2023_PreFX}
  MaterialFunctionCall(Material=NamedRerouteUsage(), M=@n1.M, Pre Skinned Local Position=@n1.M, Distance Blend=@n1.M, Fresnel=@n1.M) {materialFunction=&MF_CharacterParent_2023_PostFX}
  n2 = NamedRerouteUsage()
  n3 = DotProduct(A=CameraVectorWS(), B=Normalize(VectorInput=Transform(None=@n2)))
  n4 = NamedRerouteDeclaration(None=Divide(A=ArccosineFast!(None=@n3)) {constB=1.5707}) {name=LinearFresnel, nodeColor=<rgba 0.4531 0 1 1>}
  n5 = ScalarParameter() {defaultValue=0.624, parameterName=IridesenceUVExponent}
  n6 = Add(
   A=Multiply(A=Power(Base=@n4, Exp=@n5), B=ScalarParameter() {defaultValue=2.6501, parameterName=IridesenceUVScale})
   B=ScalarParameter() {defaultValue=1.3275, parameterName=IridesenceUVOffset}
  )
  n7 = Constant() {r=0.5}
  n8 = NamedRerouteDeclaration(None=TextureSample(UVs=MaterialFunctionCall(X=@n6, Y=@n7) {materialFunction=&MakeFloat2}.Result) {texture=/Game/Global/Textures/Default/Thinfilm/T_ThinFilm_Spectrum_BlueGreen_Color}.RGB) {name=IridesenceColor, nodeColor=<rgba 1 0.7969 0 1>}
  Multiply(A=@n8, B=ScalarParameter() {defaultValue=0.2, parameterName=IridesenceStrength})
  n9 = TextureSample(UVs=NamedRerouteUsage()) {
   samplerType=SAMPLERTYPE_LinearColor
   texture=/BRCosmetics/Characters/Player/Female/Medium/Bodies/F_MED_ElegantLilyCharm/Textures/T_ElegantLilyCharm_Body_FX2
  }
  NamedRerouteDeclaration(None=@n9.A) {name=StripeMask, nodeColor=<rgba 1 0.9844 0 1>}
  n10 = Multiply(
   A=Power(Base=NamedRerouteUsage(), Exp=ScalarParameter() {defaultValue=1, parameterName=Exp})
   B=ScalarParameter() {defaultValue=0.5, parameterName=SparkleEmissive}
  )
  n11 = NamedRerouteUsage()
  n12 = NamedRerouteUsage()
  n13 = ScalarParameter() {defaultValue=1, parameterName=EmissiveStrength}
  n14 = Power(
   Base=MaterialFunctionCall() {materialFunction=/Game/Packages/Fortress_SharedMaterials/Base_Material_Functions/BaseCharacterShaderFunctions/Audio/MF_AudioReactivity}.Amplitude
   Exp=ScalarParameter() {defaultValue=3, parameterName=AmplitudeExponent}
  )
  n15 = ScalarParameter() {defaultValue=1, parameterName=FrontEndReactivityBoost}
  n16 = Multiply(
   A=Add(A=Multiply(A=@n10, B=@n11), B=Multiply(A=@n12, B=@n13))
   B=NamedRerouteDeclaration(None=Multiply(A=@n14, B=@n15)) {name=Audio, nodeColor=<rgba 0.8281 0 1 1>}
  )
  n17 = ScalarParameter() {defaultValue=-0.5, parameterName=EmissiveDesat}
  n18 = Saturate(None=ScalarParameter() {defaultValue=1, parameterName=PlatinumReactivity})
  NamedRerouteDeclaration(None=Multiply(A=Desaturation(None=@n16, Fraction=@n17), B=NamedRerouteDeclaration(None=@n18) {name=PlatinumReactivity, nodeColor=<rgba 0.5469 1 0 1>})) {name=Emissive, nodeColor=<rgba 0.875 0 1 1>}
  n19 = PreSkinnedLocalBounds()
  n20 = Subtract(A=VertexInterpolator(VS=LocalPosition() {localOrigin=InstancePreSkinning}.Z).PS, B=ComponentMask(None=@n19.Min) {b=true})
  n21 = ComponentMask(None=@n19.Extents) {b=true}
  NamedRerouteDeclaration(None=Step(Y=Divide(A=@n20, B=@n21)) {constX=0.47, constY=0.5}) {name=BootMask, nodeColor=<rgba 1 0.9609 0 1>}
  n22 = Add(
   A=Multiply(A=NamedRerouteUsage(), B=ScalarParameter() {defaultValue=0.2, parameterName=bootfresnelbias})
   B=ScalarParameter() {defaultValue=0.4, parameterName=IridesenceMaskThreshold}
  )
  n23 = ScalarParameter() {defaultValue=0.1, parameterName=IridesenceMaskBias}
  n24 = Saturate(None=InverseLinearInterpolate(A=@n22, B=Add(A=@n22, B=@n23), Value=NamedRerouteUsage()))
  n25 = Multiply(A=@n24, B=@n24)
  NamedRerouteDeclaration(None=Multiply(A=@n8, B=NamedRerouteDeclaration(None=@n25) {name=IridescenceMask, nodeColor=<rgba 0.5703 1 0 1>})) {name=Iridesence, nodeColor=<rgba 1 0 0.9141 1>}
  n26 = Multiply(
   A=Power(Base=NamedRerouteUsage(), Exp=ScalarParameter() {defaultValue=0.8, parameterName=PlatinumUVExponent})
   B=ScalarParameter() {defaultValue=1, parameterName=PlatinumUVScale}
  )
  n27 = ScalarParameter() {defaultValue=0.4, parameterName=PlatinumUVOffset}
  n28 = TextureSample(UVs=MaterialFunctionCall(X=Add(A=@n26, B=@n27), Y=Constant() {r=0.5}) {materialFunction=&MakeFloat2}.Result) {texture=/BRCosmetics/Characters/Player/Male/Medium/Bodies/M_MED_TofuResort/Textures/T_Platinum_Gradient}
  n29 = ScalarParameter() {defaultValue=0.8, parameterName=PlatinumStrength}
  n30 = ScalarParameter() {parameterName=PlatinumMaskThreshold}
  n31 = ScalarParameter() {defaultValue=0.1, parameterName=PlatinumMaskBias}
  n32 = Saturate(None=InverseLinearInterpolate(A=@n30, B=Add(A=@n30, B=@n31), Value=NamedRerouteUsage()))
  NamedRerouteDeclaration(None=Multiply(A=Multiply(A=@n28.RGB, B=@n29), B=Multiply(A=@n32, B=@n32))) {name=PlatinumGradient, nodeColor=<rgba 1 0 0.7734 1>}
  n33 = TextureObject() {
   samplerType=SAMPLERTYPE_LinearColor
   texture=/BRCosmetics/Global/Materials/MaterialGym/Materials/AdvancedSequins/Textures/Substance_graph_T_sequin_UV
  }
  n34 = MaterialFunctionCall!(
   Texture=@n33
   TilingXY=NamedRerouteDeclaration(None=ScalarParameter() {defaultValue=22, parameterName=PatternTiling}) {name=PatternTiling, nodeColor=<rgba 1 0 0.0234 1>}
   OffsetXY=ComponentMask(None=VectorParameter() {parameterName=UVOffset, defaultValue=<rgba 0 0 0 1>}.RGB) {g=true, r=true}
   AnimateOffset=StaticBool() {value=true}
   UseParallax
   Height
  ) {materialFunction=/Game/Global/Materials/MaterialGym/Materials/TechArtPatterns/MaterialFunctions/MF_TextureUVs}
  NamedRerouteDeclaration(None=@n34.TextureUVs) {name=TextureBasedUV, desc=Texture based UV, nodeColor=<rgba 0 0.4844 1 1>}
  n35 = TextureSample!(UVs=NamedRerouteUsage(), Tex, Bias=Constant() {r=-1}, Apply View MipBias) {
   mipValueMode=TMVM_MipBias
   samplerType=SAMPLERTYPE_Normal
   texture=/BRCosmetics/Global/Materials/MaterialGym/Materials/AdvancedSequins/Textures/Substance_graph_T_sequin_N
  }
  n36 = MaterialFunctionCall(Normal=@n35.RGB, Flatness=ScalarParameter() {defaultValue=0.6, parameterName=SequinNormalFlatness}) {materialFunction=&FlattenNormal}
  n37 = NamedRerouteUsage()
  n38 = NamedRerouteUsage()
  n39 = NamedRerouteDeclaration(None=Saturate(None=Multiply(A=@n37, B=@n38) {constA=1})) {name=SequinNoise, nodeColor=<rgba 0 0.25 1 1>}
  n40 = MaterialFunctionCall(
   Normal=Transform(None=Constant3Vector() {constant=<rgba 0 0 1 0>}) {transformSourceType=TRANSFORMSOURCE_World, transformType=TRANSFORM_Tangent}
   Flatness=Add(A=ScalarParameter() {defaultValue=0.3, parameterName="Flatness (S)"}, B=OneMinus(None=@n39))
  ) {materialFunction=&FlattenNormal}
  NamedRerouteDeclaration(None=LinearInterpolate(
    A=@n36.Result
    B=MaterialFunctionCall(BaseNormal=@n36.Result, AdditionalNormal=@n40.Result) {materialFunction=&BlendAngleCorrectedNormals}.Result
    Alpha=NamedRerouteUsage()
   )) {name=Normal, nodeColor=<rgba 0 0.3437 1 1>}
  NamedRerouteDeclaration(None=@n1.Fresnel) {name=Fresnel, nodeColor=<rgba 0 1 0.0625 1>}
  n41 = TextureSample!(UVs=NamedRerouteUsage(), Tex, Apply View MipBias) {
   samplerType=SAMPLERTYPE_Masks
   texture=/BRCosmetics/Global/Materials/MaterialGym/Materials/AdvancedSequins/Textures/Substance_graph_T_sequin_M
  }
  NamedRerouteDeclaration(None=@n41.B) {name=SequinMask, nodeColor=<rgba 1 0.6094 0 1>}
  n42 = ScalarParameter() {defaultValue=2, parameterName=ReactiveStripeOffset}
  n43 = NamedRerouteUsage()
  n44 = Multiply(A=Time(), B=ScalarParameter() {defaultValue=0.5, parameterName=ReactiveStripeSpeed})
  n45 = NamedRerouteUsage()
  n46 = Multiply(A=Add(A=Multiply(A=@n42, B=@n43), B=Add(A=@n44, B=@n45)), B=ScalarParameter() {defaultValue=2, parameterName=ReactiveStripeFrequency})
  NamedRerouteDeclaration(None=ConstantBiasScale(None=Sine(None=@n46))) {name=StripeMovement, nodeColor=<rgba 0 0.9766 1 1>}
  NamedRerouteDeclaration(None=@n34.BaseUVs) {name=BaseUVs, nodeColor=<rgba 0.9922 1 0 1>}
  n47 = ConstantBiasScale(None=Sine(None=Time()))
  n48 = ScalarParameter() {parameterName=RandomOffsetStrength}
  NamedRerouteDeclaration(None=Multiply(A=Multiply(A=@n47, B=@n48), B=NamedRerouteUsage())) {name=RandomUVOffset, nodeColor=<rgba 0 0.4844 1 1>}
  NamedRerouteDeclaration(None=LinearInterpolate(A=ScalarParameter() {defaultValue=0.4, parameterName=SpecMin}, B=ScalarParameter() {defaultValue=0.6, parameterName=SpecMax}, Alpha=@n41.R)) {name=Spec, nodeColor=<rgba 0 1 0.3672 1>}
  NamedRerouteDeclaration(None=LinearInterpolate(
    A=ScalarParameter() {defaultValue=0.2, parameterName=RoughnessMin}
    B=ScalarParameter() {defaultValue=0.5, parameterName=RoughnessMax}
    Alpha=@n41.G
   )) {name=Roughness, nodeColor=<rgba 0 1 0.3672 1>}
  n49 = NamedRerouteUsage()
  n50 = NamedRerouteUsage()
  NamedRerouteDeclaration(None=TextureSample(UVs=Add(A=@n49, B=@n50)) {texture=/Game/Global/Textures/Default/Masks_and_Gradients/chunkma_2}.RGB) {name=Test, nodeColor=<rgba 0 1 0.3438 1>}
  NamedRerouteDeclaration(None=@n41.R) {name=Random, nodeColor=<rgba 0.3828 0 1 1>}
  n51 = SetMaterialAttributes!(
   MaterialAttributes=NamedRerouteUsage()
   Normal=MaterialFunctionCall(BaseNormal=NamedRerouteUsage(), AdditionalNormal=NamedRerouteUsage()) {materialFunction=&BlendAngleCorrectedNormals}.Result
   Roughness=NamedRerouteUsage()
   Metallic=ScalarParameter() {defaultValue=0.5, parameterName=Metallic}
   Shading Model=ShadingModel()
   Base Color=NamedRerouteUsage()
   Specular=NamedRerouteUsage()
   Emissive Color=NamedRerouteUsage()
  ) {attributeSetTypes [
    #1
    #2
    #3
    #4
    #5
    #6
    #7
   ]}
  n52 = NamedRerouteUsage()
  NamedRerouteDeclaration(None=BlendMaterialAttributes(A=NamedRerouteUsage(), B=@n51, Alpha=ChannelMaskParameter(None=@n52) {parameterName=TechArtMaskChannel}.R)) {name=OUT_Material, nodeColor=<rgba 0.8047 1 0 1>}
  NamedRerouteDeclaration(None=@n9.G) {name=StripeOffset, nodeColor=<rgba 0.6641 0 1 1>}
  NamedRerouteDeclaration(None=GetMaterialAttributes(None=NamedRerouteDeclaration(None=@n1.Result) {name=IN_Material, nodeColor=<rgba 0.875 0 1 1>}) {attributeGetTypes [#1]}.Normal) {name=IN_Normal, nodeColor=<rgba 0 1 0.5547 1>}
  n53 = VectorParameter() {parameterName=StripeColor, defaultValue=<rgba 0.5216 0.4628 0.7344 1>}
  n54 = VectorParameter() {parameterName=StripeColorBoot, defaultValue=<rgba 0.7211 0.638 0.8333 1>}
  n55 = NamedRerouteUsage()
  n56 = ScalarParameter() {defaultValue=0.2, parameterName=StripeIridesenceAdd}
  n57 = NamedRerouteUsage()
  n58 = LinearInterpolate(A=NamedRerouteUsage(), B=Add(A=LinearInterpolate(A=@n53.RGB, B=@n54.RGB, Alpha=@n55), B=Multiply(A=@n56, B=@n57)), Alpha=NamedRerouteUsage())
  n59 = NamedRerouteUsage()
  NamedRerouteDeclaration(None=Add(A=Add(A=@n58, B=@n59), B=NamedRerouteUsage())) {name=BaseColor, nodeColor=<rgba 0.125 0 1 1>}
  n60 = NamedRerouteUsage()
  n61 = Multiply(A=NamedRerouteUsage(), B=NamedRerouteUsage())
  n62 = LinearInterpolate(
   A=ScalarParameter() {defaultValue=0.5, parameterName=StripeThickness}
   B=ScalarParameter() {defaultValue=0.1, parameterName=ReactiveStripeThickness}
   Alpha=Saturate(None=Multiply(A=@n60, B=@n61))
  )
  n63 = Add(A=@n62) {constB=0.1}
  n64 = NamedRerouteUsage()
  NamedRerouteDeclaration(None=Saturate(None=InverseLinearInterpolate(A=@n62, B=@n63, Value=@n64))) {name=Stripes, nodeColor=<rgba 1 0.7031 0 1>}
  NamedRerouteDeclaration(None=MaterialFunctionCall(
    Color_Morning=VectorParameter() {parameterName=Color_Morning, defaultValue=<rgba 0.2 0.2 0.2 1>}.RGB
    Color_Day=VectorParameter() {parameterName=Color_Day, defaultValue=<rgba 0.075 0.075 0.075 1>}.RGB
    Color_Evening=VectorParameter() {parameterName=Color_Evening, defaultValue=<rgba 0.2 0.2 0.2 1>}.RGB
    Color_Night=VectorParameter() {parameterName=Color_Night, defaultValue=<rgba 0.2 0.2 0.2 1>}.RGB
   ) {materialFunction=&Color_Time_of_Day}.Result) {name=Color, nodeColor=<rgba 1 0.4453 0 1>}
  n65 = NamedRerouteUsage()
  n66 = ScalarParameter() {defaultValue=2.25, parameterName=NoiseTile}
  n67 = Constant()
  n68 = ScalarParameter() {defaultValue=-0.15, parameterName=SequinNoiseMaskSpeed}
  n69 = TextureSample(UVs=Panner(
    Coordinate=Multiply(A=@n65, B=@n66)
    Speed=MaterialFunctionCall(X=@n67, Y=@n68) {materialFunction=&MakeFloat2}.Result
   )) {samplerType=SAMPLERTYPE_Grayscale, texture=/Game/Global/Textures/Default/Noise/LowResBlurredNoiseNormalized}
  n70 = Add(
   A=Multiply(A=@n69.R, B=ScalarParameter() {defaultValue=1, parameterName=NoiseMultiplier})
   B=ScalarParameter() {defaultValue=0.3, parameterName=NoiseBias}
  )
  NamedRerouteDeclaration(None=Power(Base=Saturate(None=@n70), Exp=ScalarParameter() {defaultValue=1, parameterName=NoiseExponent})) {name=SequinNoiseMask, nodeColor=<rgba 1 0 0.6328 1>}
  NamedRerouteDeclaration(None=@n69.R) {name=EmissiveSequinNoise, nodeColor=<rgba 1 0 0.6328 1>}
  NamedRerouteDeclaration(None=@n9.B) {name=StripeGradient, nodeColor=<rgba 0 1 0.3438 1>}
  NamedRerouteDeclaration(None=@n1.Tech Art Mask) {name=FXMask, nodeColor=<rgba 0 0.5078 1 1>}
  VectorParameter() {parameterName=BaseColor, defaultValue=<rgba 0.075 0.075 0.075 1>}
 ]
}
